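# Unlit translucent Niagara sprite material for the BlazerVeil emote, drawing a radial mask tinted by particle color.
# Particle RGB passes through MF_FX_EmissiveAdjustments for emissive, while particle alpha multiplies a centered SphereMask for opacity.
# The mask's UVs are scaled about their center by the particle's width-to-height ratio.
# The Radius dynamic parameter drives both mask radius and hardness, leaving the named Hardness parameter unused.
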Material /BRCosmetics/Animation/Game/MainPlayer/Emotes/BlazerVeil/FX/M_Emote_BlazerVeil_EmissiveReveal
graph {
 properties {bUsedWithNiagaraSprites=true, blendMode=BLEND_Translucent, pixelDepthOffsetMode=PDOM_Legacy, shadingModel=MSM_Unlit, usedShadingModels=Unlit}
 nodes [
  n1 = ParticleColor()
  MaterialFunctionCall(Emissive=@n1.RGB) {materialFunction=&MF_FX_EmissiveAdjustments}
  n2 = TextureCoordinate()
  n3 = ParticleSize()
  n4 = AppendVector(A=Constant() {r=1}, B=Divide(A=ComponentMask(None=@n3) {r=true}, B=ComponentMask(None=@n3) {g=true}))
  n5 = DynamicParameter() {defaultValue=<rgba 0.5 0 1 1>, paramNames [Radius,Hardness,Param3,Param4]}
  Multiply(
   A=@n1.A
   B=SphereMask(
    A=MaterialFunctionCall(UVs=@n2, Texture Scale=@n4) {materialFunction=&ScaleUVsByCenter}.UVs
    B=Constant() {r=0.5}
    Radius=@n5.Radius
    Hardness=@n5.Radius
   )
  )
 ]
}
